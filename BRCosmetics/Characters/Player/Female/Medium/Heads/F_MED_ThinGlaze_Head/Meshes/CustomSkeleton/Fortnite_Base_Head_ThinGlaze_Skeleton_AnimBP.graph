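# Anim Blueprint for the ThinGlaze female cosmetic head that partially counter-rotates the head bone on top of the input pose.
# Each update negates the head's ref-pose delta relative to neck_02, scales it per axis, and applies it as an additive bone-space Modify Bone rotation.
# The update also runs the nativized gravity override state function, storing gravity and rigid body settings, which the AnimGraph does not use.
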AnimBlueprint /BRCosmetics/Characters/Player/Female/Medium/Heads/F_MED_ThinGlaze_Head/Meshes/CustomSkeleton/Fortnite_Base_Head_ThinGlaze_Skeleton_AnimBP
target_skeleton=&/Fortnite_Base_Head_ThinGlaze_Skeleton
graphs [
 AnimGraph {nodes [
   n1 = Local To Component(LocalPose=Input Pose())
   n2 = Get head rot() -> (head rot:Rotator)
   Output Pose(Result=Component To Local(ComponentPose="Transform (Modify) Bone - Bone: head"!(ComponentPose=@n1, Alpha=1, Rotation=@n2) -> (Pose) {rotationMode=BMM_Additive, rotationSpace=BCS_BoneSpace, boneToModify {boneName=head}}))
  ]}
 EventGraph {nodes [
   Event BlueprintUpdateAnimation()
   > n1 = "Gravity Override State Function (Nativized)"!(
    execute
    CharacterPartAnimInstance:Custom Character Part Anim Instance Object Reference=Self-Reference()
    GravityOverrideParameterSet:Fort Gravity Override Parameter Set Structure=Get Gravity Override Params_0() -> (Gravity Override Params_0:Fort Gravity Override Parameter Set Structure)
    RigidBodyAnimNodeParameterSet:Fort Rigid Body Anim Node Parameter Set Structure=Get RBAN Params_0() -> (RBAN Params_0:Fort Rigid Body Anim Node Parameter Set Structure)
    DebugDraw:bool=false
    GameTimeInSeconds:float=GetGameTimeInSeconds()
    DebugJointName:Name=root
    StateLogicLayerName:Name=none
   ) -> (
    then
    OutGravityOverride:Vector
    OutCurrentState:Name=None
    OutGravityOverrideParameters:Fort Gravity Override Parameters Structure
    OutRigidBodyAnimNodeParameters:Fort Rigid Body Anim Node Parameters Structure
   )
   > n2 = Set Gravity Override(Gravity Override=@n1)
   > Set Gravity Override Cape(Gravity Override Cape:Vector=@n2) -> (Output_Get:Vector)
   > Set RBAN Settings Out_0(RBAN Settings Out_0:Fort Rigid Body Anim Node Parameters Structure=@n1.OutRigidBodyAnimNodeParameters) -> (Output_Get:Fort Rigid Body Anim Node Parameters Structure)
   n3 = GetOwnerSkeletalMeshComponent()
   n4 = BreakRotator(InRot=BreakTransform(InTransform=GetDeltaTransformFromRefPose(self=@n3, BoneName=head, BaseName=neck_02)).Rotation)
   n5 = Get Y percentage() -> (Y percentage:float)
   n6 = Get X percentage() -> (X percentage:float)
   n7 = Get Z percentage() -> (Z percentage:float)
   > Set head rot(head rot:Rotator=MakeRotator(
     Roll=float * float!(A=@n4, B=-1, C:double=@n6) -> (ReturnValue)
     Pitch=float * float(A=@n4.Pitch, B=-1, C=@n5)
     Yaw=float * float(A=@n4.Yaw, B=-1, C=@n7)
    )) -> (Output_Get:Rotator)
   TryGetPawnOwner()
  ]}
]
variables [
 Gravity Override {type=Vector}
 Gravity Override Cape {type=Vector}
 Gravity Override Params_0 {
  type=FortGravityOverrideParameterSet
  default=(Skydive_GoingUp=(GravityOverride=()),Skydive_FloatingDown=(GravityOverride=()),Skydive_DivingDown=(GravityOverride=()),Skydive_Parachuting=(GravityOverride=()),OnGround_Standing=(GravityOverride=()),OnGround_Moving=(GravityOverride=()),OnGround_Moving_HighRotationRate=(GravityOverride=()),OnGround_Crouch_Moving=(GravityOverride=()),Emote_or_Melee=(GravityOverride=()),Swimming=(GravityOverride=()),Front_End=(GravityOverride=()),DBNO=(GravityOverride=()),DBNO_Carried=(GravityOverride=()),OnGround_Falling=(GravityOverride=()),Ziplining=(GravityOverride=()),Skydive_Surfing=(GravityOverride=()),Riding_Standing=(GravityOverride=()),Riding_Moving=(GravityOverride=()),Riding_Moving_Targeting=(GravityOverride=()),Tactical_Sprint=(GravityOverride=()),Rail_Grinding=(GravityOverride=()),)
 }
 head rot {type=Rotator}
 RBAN Params 0 {
  type=FortRigidBodyAnimNodeParameterSet
  default=(Skydive_GoingUp=(RigidBodyAnimNode=()),Skydive_FloatingDown=(RigidBodyAnimNode=()),Skydive_DivingDown=(RigidBodyAnimNode=()),Skydive_Parachuting=(RigidBodyAnimNode=()),OnGround_Standing=(RigidBodyAnimNode=()),OnGround_Moving=(RigidBodyAnimNode=()),OnGround_Moving_HighRotationRate=(RigidBodyAnimNode=()),OnGround_Crouch_Moving=(RigidBodyAnimNode=()),Emote_or_Melee=(RigidBodyAnimNode=()),Swimming=(RigidBodyAnimNode=()),Front_End=(RigidBodyAnimNode=()),DBNO=(RigidBodyAnimNode=()),DBNO_Carried=(RigidBodyAnimNode=()),OnGround_Falling=(RigidBodyAnimNode=()),Ziplining=(RigidBodyAnimNode=()),Skydive_Surfing=(RigidBodyAnimNode=()),Riding_Standing=(RigidBodyAnimNode=()),Riding_Moving=(RigidBodyAnimNode=()),Riding_Moving_Targeting=(RigidBodyAnimNode=()),Tactical_Sprint=(RigidBodyAnimNode=()),Rail_Grinding=(RigidBodyAnimNode=()),)
 }
 RBAN Params 0_0 {
  type=FortRigidBodyAnimNodeParameterSet
  default=(Skydive_GoingUp=(RigidBodyAnimNode=()),Skydive_FloatingDown=(RigidBodyAnimNode=()),Skydive_DivingDown=(RigidBodyAnimNode=()),Skydive_Parachuting=(RigidBodyAnimNode=()),OnGround_Standing=(RigidBodyAnimNode=()),OnGround_Moving=(RigidBodyAnimNode=()),OnGround_Moving_HighRotationRate=(RigidBodyAnimNode=()),OnGround_Crouch_Moving=(RigidBodyAnimNode=()),Emote_or_Melee=(RigidBodyAnimNode=()),Swimming=(RigidBodyAnimNode=()),Front_End=(RigidBodyAnimNode=()),DBNO=(RigidBodyAnimNode=()),DBNO_Carried=(RigidBodyAnimNode=()),OnGround_Falling=(RigidBodyAnimNode=()),Ziplining=(RigidBodyAnimNode=()),Skydive_Surfing=(RigidBodyAnimNode=()),Riding_Standing=(RigidBodyAnimNode=()),Riding_Moving=(RigidBodyAnimNode=()),Riding_Moving_Targeting=(RigidBodyAnimNode=()),Tactical_Sprint=(RigidBodyAnimNode=()),Rail_Grinding=(RigidBodyAnimNode=()),)
 }
 RBAN Params_0 {
  type=FortRigidBodyAnimNodeParameterSet
  default=(Skydive_GoingUp=(RigidBodyAnimNode=()),Skydive_FloatingDown=(RigidBodyAnimNode=()),Skydive_DivingDown=(RigidBodyAnimNode=()),Skydive_Parachuting=(RigidBodyAnimNode=()),OnGround_Standing=(RigidBodyAnimNode=()),OnGround_Moving=(RigidBodyAnimNode=()),OnGround_Moving_HighRotationRate=(RigidBodyAnimNode=()),OnGround_Crouch_Moving=(RigidBodyAnimNode=()),Emote_or_Melee=(RigidBodyAnimNode=()),Swimming=(RigidBodyAnimNode=()),Front_End=(RigidBodyAnimNode=()),DBNO=(RigidBodyAnimNode=()),DBNO_Carried=(RigidBodyAnimNode=()),OnGround_Falling=(RigidBodyAnimNode=()),Ziplining=(RigidBodyAnimNode=()),Skydive_Surfing=(RigidBodyAnimNode=()),Riding_Standing=(RigidBodyAnimNode=()),Riding_Moving=(RigidBodyAnimNode=()),Riding_Moving_Targeting=(RigidBodyAnimNode=()),Tactical_Sprint=(RigidBodyAnimNode=()),Rail_Grinding=(RigidBodyAnimNode=()),)
 }
 RBAN Settings Out_0 {type=FortRigidBodyAnimNodeParameters}
 X percentage {type=number}
 Y percentage {type=number, default=0.5}
 Z percentage {type=number, default=0.8}
]
